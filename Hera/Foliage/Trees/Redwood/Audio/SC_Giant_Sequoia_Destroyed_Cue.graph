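# Sound cue for a giant sequoia being destroyed, played in the impacts sound class with concurrency capped at four.
# A close-attenuated layer mixes pitch-randomized destruction and break one-shots with a quieter-mixed, briefly delayed, volume-boosted distant layer under its own attenuation.
# A quality-level root sends its first two slots to the full random wave pools and its last two to an identical chain with only two variations per pool.
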SoundCue /Hera/Foliage/Trees/Redwood/Audio/SC_Giant_Sequoia_Destroyed_Cue
root=QualityLevel_1
nodes [
 {
  id=Attenuation_3
  children [Mixer_0]
  properties {attenuationSettings=/Game/Sounds/Fort_Impact_Sounds/Demolition/Tree/Demo_Tree_Large_Destroyed_Close_Att}
 }
 {
  id=Attenuation_4
  children [Delay_2]
  properties {attenuationSettings=/Game/Sounds/Fort_Impact_Sounds/Demolition/Tree/Demo_Tree_Large_Destroyed_Distant_Att}
 }
 {
  id=Attenuation_5
  children [Mixer_6]
  properties {attenuationSettings=&Demo_Tree_Large_Destroyed_Close_Att}
 }
 {
  id=Attenuation_7
  children [Delay_1]
  properties {attenuationSettings=&Demo_Tree_Large_Destroyed_Distant_Att}
 }
 {id=Delay_1, children [Modulator_8], properties {delayMax=0.3, delayMin=0.3}}
 {id=Delay_2, children [Modulator_14], properties {delayMax=0.3, delayMin=0.3}}
 {id=Mixer_0, children [Modulator_3,Modulator_9], properties {inputVolume [0.6,0.8]}}
 {id=Mixer_3, children [Attenuation_3, Attenuation_7], properties {inputVolume [1,0.4]}}
 {id=Mixer_6, children [Modulator_13,Modulator_15], properties {inputVolume [0.6,0.8]}}
 {id=Mixer_7, children [Attenuation_5, Attenuation_4], properties {inputVolume [1,0.4]}}
 {id=Modulator_13, children [Random_15], properties {pitchMax=1.1, pitchMin=0.9}}
 {id=Modulator_14, children [Random_16], properties {volumeMax=2, volumeMin=2}}
 {id=Modulator_15, children [Random_17], properties {pitchMax=1.3}}
 {id=Modulator_3, children [Random_4], properties {pitchMax=1.1, pitchMin=0.9}}
 {id=Modulator_8, children [Random_6], properties {volumeMax=2, volumeMin=2}}
 {id=Modulator_9, children [Random_11], properties {pitchMax=1.3}}
 {id=QualityLevel_1, children [Mixer_3,Mixer_3,Mixer_7,Mixer_7]}
 {
  id=Random_11
  children [WavePlayer_54, WavePlayer_55, WavePlayer_56, WavePlayer_57, WavePlayer_58]
  properties {bShouldExcludeFromBranchCulling=true, weights [1,1,1,1,1]}
 }
 {id=Random_15, children [WavePlayer_73, WavePlayer_74], properties {weights [1,1]}}
 {id=Random_16, children [WavePlayer_76, WavePlayer_77], properties {weights [1,1]}}
 {id=Random_17, children [WavePlayer_82, WavePlayer_83], properties {bShouldExcludeFromBranchCulling=true, weights [1,1]}}
 {id=Random_4, children [WavePlayer_29, WavePlayer_30, WavePlayer_33], properties {weights [1,1,1]}}
 {
  id=Random_6
  children [WavePlayer_35, WavePlayer_36, WavePlayer_37, WavePlayer_38, WavePlayer_39, WavePlayer_40]
  properties {weights [1,1,1,1,1,1]}
 }
 {id=WavePlayer_29, properties {soundWaveAssetPtr=/AmbientAudioContent/Waves/OneShots/Spot_PhysicsTree_Destroyed_01}}
 {id=WavePlayer_30, properties {soundWaveAssetPtr=/AmbientAudioContent/Waves/OneShots/Spot_PhysicsTree_Destroyed_02}}
 {id=WavePlayer_33, properties {soundWaveAssetPtr=/AmbientAudioContent/Waves/OneShots/Spot_PhysicsTree_Destroyed_03}}
 {id=WavePlayer_35, properties {soundWaveAssetPtr=/AmbientAudioContent/Waves/OneShots/Spot_PhysicsTree_Destroyed_Distant_01}}
 {id=WavePlayer_36, properties {soundWaveAssetPtr=/AmbientAudioContent/Waves/OneShots/Spot_PhysicsTree_Destroyed_Distant_02}}
 {id=WavePlayer_37, properties {soundWaveAssetPtr=/AmbientAudioContent/Waves/OneShots/Spot_PhysicsTree_Destroyed_Distant_03}}
 {id=WavePlayer_38, properties {soundWaveAssetPtr=/AmbientAudioContent/Waves/OneShots/Spot_PhysicsTree_Destroyed_Distant_04}}
 {id=WavePlayer_39, properties {soundWaveAssetPtr=/AmbientAudioContent/Waves/OneShots/Spot_PhysicsTree_Destroyed_Distant_06}}
 {id=WavePlayer_40, properties {soundWaveAssetPtr=/AmbientAudioContent/Waves/OneShots/Spot_PhysicsTree_Destroyed_Distant_05}}
 {id=WavePlayer_54, properties {soundWaveAssetPtr=/Game/Sounds/Fort_Impact_Sounds/Demolition/Tree/Demo_Tree_Large_Destroyed_Break_01}}
 {id=WavePlayer_55, properties {soundWaveAssetPtr=/Game/Sounds/Fort_Impact_Sounds/Demolition/Tree/Demo_Tree_Large_Destroyed_Break_02}}
 {id=WavePlayer_56, properties {soundWaveAssetPtr=/Game/Sounds/Fort_Impact_Sounds/Demolition/Tree/Demo_Tree_Large_Destroyed_Break_03}}
 {id=WavePlayer_57, properties {soundWaveAssetPtr=/Game/Sounds/Fort_Impact_Sounds/Demolition/Tree/Demo_Tree_Large_Destroyed_Break_04}}
 {id=WavePlayer_58, properties {soundWaveAssetPtr=/Game/Sounds/Fort_Impact_Sounds/Demolition/Tree/Demo_Tree_Large_Destroyed_Break_05}}
 {id=WavePlayer_73, properties {soundWaveAssetPtr=&Spot_PhysicsTree_Destroyed_01}}
 {id=WavePlayer_74, properties {soundWaveAssetPtr=&Spot_PhysicsTree_Destroyed_02}}
 {id=WavePlayer_76, properties {soundWaveAssetPtr=&Spot_PhysicsTree_Destroyed_Distant_01}}
 {id=WavePlayer_77, properties {soundWaveAssetPtr=&Spot_PhysicsTree_Destroyed_Distant_02}}
 {id=WavePlayer_82, properties {soundWaveAssetPtr=&Demo_Tree_Large_Destroyed_Break_01}}
 {id=WavePlayer_83, properties {soundWaveAssetPtr=&Demo_Tree_Large_Destroyed_Break_02}}
]
properties {
 bOverrideConcurrency=true
 maxDistance=18000
 soundClassObject=&Fort_Impacts
 volumeMultiplier=0.56
 concurrencyOverrides {maxCount=4}
 concurrencySet [&CONC_Object_Impacts_FourMax]
}
